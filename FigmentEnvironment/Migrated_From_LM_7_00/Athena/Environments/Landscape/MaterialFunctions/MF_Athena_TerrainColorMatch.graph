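# Material function that computes a terrain-matched grass color for the Athena landscape, judging by its names.
# It samples the combined terrain color texture with UVs from the terrain coordinates function, using a mip bias for a low-detail read.
# The sample is scaled by ColormapDivisor from the landscape parameter collection, then tinted by the NeutralGrassColor parameter.
# Outputs expose the tinted result, the neutral grass color alone, and the scaled sample without tint.
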MaterialFunction /FigmentEnvironment/Migrated_From_LM_7_00/Athena/Environments/Landscape/MaterialFunctions/MF_Athena_TerrainColorMatch
graph {nodes [
  n1 = MaterialFunctionCall!(WorldPosition) {materialFunction=/FigmentEnvironment/Migrated_From_LM_7_00/Athena/Environments/Landscape/MaterialFunctions/Utilities/MF_Athena_Terrain_Coords}
  n2 = TextureSampleParameter2D!(UVs=@n1.WithInterpolator, Bias, Apply View MipBias) {
   constMipValue=5
   group=TerrainColorMatch
   mipValueMode=TMVM_MipBias
   parameterName=GrassTextureToMatch
   samplerType=SAMPLERTYPE_LinearColor
   texture=/FigmentEnvironment/Migrated_From_LM_7_00/Athena/Environments/Landscape/Textures/T_Athena_Terrain_CombinedColors_01
  }
  n3 = Multiply(A=@n2.RGB, B=CollectionParameter() {collection=/Game/Athena/Environments/Landscape/MPC/MPC_Landscape_S07, parameterName=ColormapDivisor})
  n4 = VectorParameter() {parameterName=NeutralGrassColor, defaultValue=<rgba 0.0675 0.145 0.02 1>}
  FunctionOutput(None=Multiply(A=@n3, B=@n4.RGB))
  FunctionOutput(None=@n4.RGB) {outputName=GrassColor}
  FunctionOutput(None=@n3) {outputName=GrassNoColor}
 ]}
