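# Material function that builds holiday snowfall masks for Athena landscape materials, driven by the Holiday_SnowAmount collection parameter.
# A world-mapped topo mask alpha is scaled by that amount, biased and optionally multiplied by an extra snow mask.
# By default HeightLerp shapes SnowMask, with plain contrast as the static-switch alternative; SnowMask_RockSubtract removes rock.
# Simple_SnowMask skips height blending and can weight snow by world-normal Z for imposters.
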MaterialFunction /FigmentEnvironment/Migrated_From_LM_7_00/Athena/Environments/Landscape/MaterialFunctions/Snow/MF_Athena_Holiday_SnowFall
graph {nodes [
  n1 = MaterialFunctionCall!(WorldPosition=FunctionInput(Preview=WorldPosition().XYZ) {bUsePreviewValueAsDefault=true, inputName=WorldPosition, sortPriority=10}) {materialFunction=/FigmentEnvironment/Migrated_From_LM_7_00/Athena/Environments/Landscape/MaterialFunctions/Utilities/MF_Athena_Terrain_Coords}
  n2 = TextureSample(UVs=FunctionInput(Preview=@n1.Result) {bUsePreviewValueAsDefault=true, inputName=UV_WorldMask, inputType=FunctionInput_Vector2}) {
   samplerSource=SSM_Wrap_WorldGroupSettings
   samplerType=SAMPLERTYPE_LinearColor
   texture=/FigmentEnvironment/Migrated_From_LM_7_00/Athena/Environments/Landscape/Textures/T_Athena_Terrain_Topo_Mask
  }
  n3 = CollectionParameter() {collection=/Game/Athena/Environments/Landscape/MPC/MPC_Landscape_S07, parameterName=Holiday_SnowAmount}
  n4 = Clamp(None=Subtract(A=Multiply(A=@n2.A, B=@n3), B=ScalarParameter() {defaultValue=0.04, parameterName=SnowStartBias}))
  n5 = FunctionInput() {bUsePreviewValueAsDefault=true, inputName=Snow_Mask, inputType=FunctionInput_Scalar, sortPriority=3, previewValue=<1 1 1 1>}
  n6 = StaticBool()
  n7 = Clamp(None=StaticSwitch(
    True=Multiply(A=@n4, B=@n5)
    False=@n4
    Value=FunctionInput(Preview=@n6) {bUsePreviewValueAsDefault=true, inputName=Use_ExtraSnowMask, inputType=FunctionInput_StaticBool}
   ))
  n8 = MaterialFunctionCall(In=@n7, Contrast=ScalarParameter() {defaultValue=8, parameterName=Snow_NormalMaskContrast}) {materialFunction=&CheapContrast}
  n9 = MaterialFunctionCall() {materialFunction=&/MF_Athena_Snow_01}
  n10 = MaterialFunctionCall(In=@n7, Contrast=Constant() {r=3}) {materialFunction=&CheapContrast}
  n11 = StaticSwitch(
   True=MaterialFunctionCall(
    Transition Phase=@n8.Result
    Height Texture=FunctionInput(Preview=@n9.Snow_Mask_Height) {bUsePreviewValueAsDefault=true, inputName=LandscapeBlend_Height, inputType=FunctionInput_Scalar, sortPriority=1}
    Contrast=ScalarParameter() {defaultValue=2, parameterName=Snow_HeightBlendContrast}
   ) {materialFunction=&HeightLerp}.Alpha
   False=@n10.Result
   Value=FunctionInput(Preview=StaticBool() {value=true}) {bUsePreviewValueAsDefault=true, inputName=UseHeightLerp, inputType=FunctionInput_StaticBool}
  )
  FeatureLevelSwitch(Default=@n11, ES2=@n10.Result, ES3_1=@n10.Result, SM4_REMOVED=@n11, SM5=@n11, SM6=@n11)
  n12 = FunctionOutput(None=@n11) {outputName=SnowMask}
  n13 = FunctionInput() {bUsePreviewValueAsDefault=true, inputName=Landscape_RockMask, inputType=FunctionInput_Scalar, sortPriority=2}
  FunctionOutput(None=Clamp(None=Subtract(A=@n12, B=@n13))) {outputName=SnowMask_RockSubtract}
  n14 = FunctionInput() {bUsePreviewValueAsDefault=true, inputName=Simple_WorldNormal, sortPriority=5, previewValue=<0 0 1 1>}
  n15 = DotProduct(A=Normalize(VectorInput=@n14), B=Constant3Vector() {constant=<rgba 0 0 1 0>})
  n16 = ScalarParameter() {parameterName=SnowNormal_Contrast}
  n17 = Multiply(
   A=Add(
    A=MaterialFunctionCall(In=@n15, Contrast=@n16) {materialFunction=&CheapContrast}.Result
    B=ScalarParameter() {defaultValue=-0.2, parameterName=SnowNormal_Bias}
   )
   B=@n8.Result
  )
  n18 = FunctionInput(Preview=StaticBool()) {
   bUsePreviewValueAsDefault=true
   description=For Imposters
   inputName=Simple_UseWorldZNormal
   inputType=FunctionInput_StaticBool
   sortPriority=4
  }
  FunctionOutput(None=Saturate(None=StaticSwitch(True=@n17, False=@n8.Result, Value=@n18))) {outputName=Simple_SnowMask}
  LinearInterpolate()
 ]}
