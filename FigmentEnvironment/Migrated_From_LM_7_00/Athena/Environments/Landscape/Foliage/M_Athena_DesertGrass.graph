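# Masked two-sided foliage material for desert grass, with a static switch for shrub leaves.
# Color blends terrain-matched sand and grass tints by topo and biome masks, with vertex alpha marking shadow areas.
# World position offset combines panning wind, Pivot Painter wind rotation, player bending, snow lowering, instance-fade sinking and near-camera push-down, zeroed on ES2/ES3.1.
# Temporal dither and debug-color nodes are left unconnected.
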Material /FigmentEnvironment/Migrated_From_LM_7_00/Athena/Environments/Landscape/Foliage/M_Athena_DesertGrass
graph {
 properties {
  bFullyRough=true
  bTangentSpaceNormal=false
  bUseMaterialAttributes=true
  bUsedWithInstancedStaticMeshes=true
  blendMode=BLEND_Masked
  pixelDepthOffsetMode=PDOM_Legacy
  shadingModel=MSM_TwoSidedFoliage
  translucencyPass=MTP_BeforeDOF
  twoSided=true
  usedShadingModels=Two Sided Foliage
 }
 nodes [
  n1 = MaterialFunctionCall!(TextureResolution=Constant2Vector() {g=1, r=1}) {materialFunction=/Engine/Functions/Engine_MaterialFunctions02/Texturing/ScreenAlignedPixelToPixelUVs}
  n2 = Custom!(None) {
   code=View.TemporalAAParams
   desc="\n*** Original source before expression upgrade ***\nView.TemporalAAParams\n*** Original source before expression upgrade ***\nFrame.TemporalAAParams"
   description=TemporalAA SampleID
   outputType=CMOT_Float2
  }
  MaterialExpressionAdd_10 = Add(A=@n1.Result, B=ComponentMask(None=@n2) {r=true})
  n3 = Custom!(p=@MaterialExpressionAdd_10) {
   code="((uint)(p.x) + 2 * (uint)(p.y)) % 5"
   outputType=CMOT_Float1
   inputs [{
     inputName=p
     input {expression=&self:MaterialExpressionAdd_10}
    }]
  }
  n4 = MaterialFunctionCall(TextureResolution=ComponentMask(None=Constant2Vector() {g=64, r=64}) {g=true, r=true}) {materialFunction=&ScreenAlignedPixelToPixelUVs}
  n5 = Add(
   A=@n3
   B=Multiply(
    A=TextureSample(UVs=@n4.Result) {samplerType=SAMPLERTYPE_LinearGrayscale, texture=/Engine/EngineMaterials/Good64x64TilingNoiseHighFreq}.R
    B=ScalarParameter() {defaultValue=1, parameterName=randin1}
   )
  ) {constB=0}
  n6 = Constant() {r=6}
  Add(
   A=Add(B=Divide(A=@n5, B=@n6) {constB=6}) {constB=0.5, desc=InputAlphaThreshold}
   B=ScalarParameter() {defaultValue=-0.5, parameterName=TempAAScale_2}
  ) {constB=-0.5}
  Constant() {r=0.05}
  Constant() {r=1200}
  Constant() {desc=RandomInput, r=1}
  Constant() {r=-0.5}
  Constant() {r=-1}
  Constant() {r=5}
  Constant() {r=50}
  Constant() {r=1}
  Constant()
  n7 = MaterialFunctionCall() {materialFunction=/FigmentEnvironment/Migrated_From_LM_7_00/Athena/Environments/Landscape/MaterialFunctions/MF_Athena_TerrainColorMatch}
  n8 = Multiply(A=VectorParameter() {parameterName=SandGrass_NeutralColor, defaultValue=<rgba 0.4621 0.3404 0.1901 1>}.RGB, B=@n7.GrassNoColor)
  n9 = Multiply(A=@n8, B=ScalarParameter() {defaultValue=1, parameterName=Base_BrightnessSand})
  n10 = Multiply(
   A=@n7.GrassColor
   B=ComponentMask(None=CollectionParameter() {collection=/Game/Athena/Environments/Landscape/MPC/MPC_Landscape_S07, parameterName=Arid_Grass_Tint}) {b=true, g=true, r=true}
  )
  n11 = Multiply(A=@n10, B=ScalarParameter() {defaultValue=0.65, parameterName=Base_BrightnessGrass})
  n12 = ComponentMask(None=WorldPosition().XYZ) {g=true, r=true}
  n13 = ScalarParameter() {defaultValue=306200, group=TerrainColorMatch, parameterName=TerrainSize}
  n14 = Divide(A=@n13)
  n15 = AppendVector(A=@n14, B=@n14)
  n16 = TextureSample(UVs=Divide(A=Subtract(A=@n12, B=@n15), B=@n13)) {
   samplerType=SAMPLERTYPE_LinearColor
   texture=/FigmentEnvironment/Migrated_From_LM_7_00/Athena/Environments/Landscape/Textures/T_Athena_Terrain_Topo_Mask
  }
  LinearInterpolate(A=@n9, B=@n11, Alpha=@n16.R)
  n17 = TextureSampleParameter2D() {parameterName=MainTexture, texture=/FigmentEnvironment/Migrated_From_LM_7_00/Environments/Foliage/Textures/T_DesertGrass_01}
  n18 = ScalarParameter() {defaultValue=0.6, parameterName=Rough}
  n19 = Constant() {r=1}
  n20 = TextureCoordinate() {coordinateIndex=1}
  n21 = Clamp(None=LinearInterpolate(
    A=ScalarParameter() {defaultValue=-0.5, parameterName=AAA111}
    B=ScalarParameter() {defaultValue=0.5, parameterName=AAA111_1}
    Alpha=ComponentMask(None=@n20) {g=true}
   ))
  n22 = LinearInterpolate(A=VertexNormalWS(), B=Constant3Vector() {constant=<rgba 0 0 1 0>}, Alpha=@n21)
  n23 = Multiply(B=ScalarParameter() {defaultValue=0.5, parameterName=shadop1_3})
  n24 = ScalarParameter() {defaultValue=2.5, parameterName=MidDistanceMaskPower}
  n25 = ScalarParameter() {defaultValue=2500, parameterName=FadeDistLength}
  n26 = ScalarParameter() {defaultValue=500, parameterName=FadeDistOffset}
  n27 = TextureSampleParameter2D(UVs=Add(A=Constant2Vector() {g=0.5}, B=TextureCoordinate()) {constB=0.2}) {parameterName=MainTexture, texture=&T_DesertGrass_01}
  n28 = PerInstanceRandom()
  n29 = MaterialFunctionCall(
   In=Divide(A=@n28, B=ScalarParameter() {defaultValue=1.85, parameterName=ShrubLeafAmounts})
   Contrast=ScalarParameter() {defaultValue=6, parameterName=ShrubLeafContrast}
  ) {materialFunction=&CheapContrast}
  n30 = StaticSwitchParameter(
   True=LinearInterpolate(A=@n17.A, B=Add(A=@n17.A, B=@n27.A), Alpha=@n29.Result)
   False=LinearInterpolate(
    A=@n17.A
    B=Power(Base=@n17.A, Exp=@n24)
    Alpha=MaterialFunctionCall(Fade Length=@n25, Fade Offset=@n26) {materialFunction=&CameraDepthFade}.Result
   )
  ) {parameterName=IsShrub}
  n31 = VertexColor()
  n32 = ScalarParameter() {defaultValue=0.5, parameterName=shadop1_2}
  n33 = TextureSample() {texture=/FigmentEnvironment/Migrated_From_LM_7_00/Environments/Foliage/Textures/T_DesertGrassShadows_01}
  n34 = ComponentMask(None=WorldPosition().XYZ) {g=true, r=true}
  n35 = Constant() {r=-1}
  n36 = ScalarParameter() {defaultValue=1000, group=WPO, parameterName=WindWaveScale}
  n37 = ScalarParameter() {defaultValue=0.1, group=WPO, parameterName=WindSpeed}
  n38 = Time()
  n39 = TextureSample(UVs=Panner(Coordinate=Divide(A=@n34, B=@n36), Time=Multiply(A=@n37, B=@n38)) {speedX=0.4, speedY=-0.5}) {
   samplerType=SAMPLERTYPE_Normal
   texture=/FigmentEnvironment/Migrated_From_LM_7_00/VisualThreatManager/StormVisuals/LargeClouds/T_StormCloudDetail01_N
  }
  n40 = Multiply(A=@n38, B=ScalarParameter() {defaultValue=0.15, group=Lightning, parameterName=WindSlowDuration})
  n41 = Sine(None=@n40)
  n42 = Multiply(
   A=@n39.RGB
   B=LinearInterpolate(
    A=ScalarParameter() {defaultValue=0.4, group=WPO, parameterName=WindPower}
    B=ScalarParameter() {defaultValue=0.1, group=WPO, parameterName=WindPowerLow}
    Alpha=ConstantBiasScale(None=@n41)
   )
  )
  n43 = ComponentMask(None=@n42) {g=true, r=true}
  n44 = Add(
   A=Divide(A=@n34, B=ScalarParameter() {defaultValue=400, group=WPO, parameterName=TerrainBlendTextureMaskScale_1})
   B=Multiply(A=ScalarParameter() {defaultValue=0.15, group=WPO, parameterName=TerrainBlendTextureMaskScale_2}, B=Multiply(A=@n35, B=@n43) {constB=-1})
  )
  n45 = Multiply(
   A=MaterialFunctionCall(
    Transition Phase=Multiply(A=@n32, B=@n33.A)
    Height Texture=TextureSample(UVs=@n44) {samplerType=SAMPLERTYPE_LinearGrayscale, texture=&Good64x64TilingNoiseHighFreq}.R
   ) {materialFunction=&HeightLerp}.Lerp Alpha No Contrast
   B=ScalarParameter() {defaultValue=1, parameterName=shadop1_4}
  )
  n46 = MaterialFunctionCall!(
   UseHeightLerp
   UV_WorldMask
   Use_ExtraSnowMask
   LandscapeBlend_Height
   Landscape_RockMask
   Snow_Mask
   Simple_UseWorldZNormal
   Simple_WorldNormal
   WorldPosition
  ) {materialFunction=/FigmentEnvironment/Migrated_From_LM_7_00/Athena/Environments/Landscape/MaterialFunctions/Snow/MF_Athena_Holiday_SnowFall}
  n47 = Multiply(
   A=AppendVector(A=Constant2Vector(), B=ScalarParameter() {defaultValue=-119.7634, parameterName=Snow_GrassLowerAmount})
   B=Clamp(None=@n46.SnowMask)
  )
  n48 = ComponentMask(None=TextureCoordinate() {coordinateIndex=1}) {g=true}
  n49 = CollectionParameter() {
   collection=/FigmentEnvironment/Migrated_From_LM_7_00/Packages/Fortress_SharedMaterials/GlobalMaterialParameters/FortniteMaterialParameters
   parameterName=PlayerPosition
  }
  n50 = MaterialFunctionCall!(
   PlayerPosition=@n49
   PlayerRadiusMax=CollectionParameter() {collection=&MPC_Landscape_S07, parameterName=PlayerBend_Radius}
   BendStrength=CollectionParameter() {collection=&MPC_Landscape_S07, parameterName=PlayerBend_Strength}
  ) {materialFunction=/FigmentEnvironment/Migrated_From_LM_7_00/Athena/Environments/Landscape/MaterialFunctions/MF_AthenaFoliagePlayerInteract}
  n51 = MaterialFunctionCall!(
   "World Position (See Note)"=MaterialFunctionCall() {materialFunction=/Engine/Functions/Engine_MaterialFunctions02/PivotPainter/PivotPainter_PerObjectData}.Pivot Position
   Max Rotation=ScalarParameter() {defaultValue=0.3, parameterName=Rotation Amount}
  ) {materialFunction=/FigmentEnvironment/Migrated_From_LM_7_00/WaterAndWind/MaterialFunctions/Fortnite_SampleRealTimeWindData}
  n52 = TextureCoordinate() {coordinateIndex=1}
  n53 = Multiply(
   A=@n51.Rotation Angle
   B=OneMinus(None=MaterialFunctionCall(Float2=@n52) {materialFunction=&BreakOutFloat2Components}.G)
  )
  n54 = MaterialFunctionCall() {materialFunction=&PivotPainter_PerObjectData}
  n55 = Subtract(A=WorldPosition().XYZ, B=@n54.Pivot Position)
  n56 = Subtract(A=WorldPosition().XYZ, B=Multiply(A=@n54.X-Axis Vector, B=DotProduct(A=@n54.X-Axis Vector, B=@n55)))
  n57 = WorldPosition() {worldPositionShaderOffset=WPT_ExcludeAllShaderOffsets}
  n58 = Constant() {r=0.5}
  n59 = Constant3Vector() {constant=<rgba 0 0 1 0>}
  n60 = TextureSample(UVs=Divide(A=@n34, B=ScalarParameter() {defaultValue=300, group=WPO, parameterName=HeightVarianceScale})) {samplerType=SAMPLERTYPE_LinearColor, texture=/FigmentEnvironment/Migrated_From_LM_7_00/Environments/Foliage/Textures/T_FieldGrass_01_M}
  n61 = LinearInterpolate(A=ScalarParameter() {group=WPO, parameterName=StaticWPOLow}, B=ScalarParameter() {group=WPO, parameterName=StaticWPOHigh}, Alpha=@n60.R)
  n62 = TextureCoordinate() {coordinateIndex=1}
  n63 = Clamp(None=Power(Base=ComponentMask(None=@n62) {g=true}, Exp=ScalarParameter() {defaultValue=1, parameterName=VertexNormalWPO-Low_1}))
  n64 = StaticSwitchParameter(True=@n63, False=Multiply(A=@n63, B=Multiply(A=@n59, B=@n61))) {parameterName=UseVertexNormalWPO}
  n65 = Add(A=@n64, B=@n57.XYZ)
  n66 = Normalize(VectorInput=AppendVector(A=@n43, B=Constant() {r=1}))
  n67 = VertexNormalWS()
  n68 = LinearInterpolate(
   A=Multiply(A=ScalarParameter() {parameterName=VertexNormalWPO-Low}, B=@n67)
   B=Multiply(A=ScalarParameter() {defaultValue=35, parameterName=VertexNormalWPO-High}, B=@n67)
   Alpha=PerInstanceRandom()
  )
  n69 = Add(
   A=StaticSwitchParameter(True=LinearInterpolate(A=@n68, B=@n67, Alpha=@n63), False=Constant()) {parameterName=UseVertexNormalWPO}
   B=Add(A=@n64, B=RotateAboutAxis(NormalizedRotationAxis=@n66, RotationAngle=@n58, PivotPoint=@n56, Position=@n65))
  )
  n70 = Add(A=@n69, B=@n57.XYZ)
  n71 = MaterialFunctionCall!(
   With RTT Features=Add(A=@n69, B=RotateAboutAxis(NormalizedRotationAxis=@n51.Rotation Axis, RotationAngle=@n53, PivotPoint=@n56, Position=@n70))
   WithOut RTT Features=@n69
   With RTT Features V2
   WithOut RTT Features V2
   With RTT Features Att
   WithOut RTT Features Att
  ) {materialFunction=/FigmentEnvironment/Migrated_From_LM_7_00/WaterAndWind/EnableOrDisableWind/Fortnite_RenderToTextureEnabled}
  n72 = Multiply(A=@n48, B=LinearInterpolate(A=@n71.V3, B=@n50.Rotation, Alpha=@n50.Rotation))
  n73 = StaticSwitchParameter(True=@n71.V3, False=LinearInterpolate(A=QualitySwitch(Default=@n72, High=@n72, Medium=@n72), B=Constant(), Alpha=VertexColor().A)) {parameterName=UseVertexNormalWPO}
  n74 = Constant2Vector()
  n75 = Multiply(A=OneMinus(None=PerInstanceFadeAmount()), B=ScalarParameter() {defaultValue=-65, parameterName=SinkIntoGroundDist})
  n76 = CameraPositionWS()
  n77 = WorldPosition()
  n78 = ScalarParameter() {defaultValue=200, parameterName=camerafadeMaskRadius}
  n79 = ScalarParameter() {parameterName=camerafadeMaskHardness}
  n80 = WorldPosition()
  n81 = CameraPositionWS()
  n82 = Multiply(
   A=Multiply(A=Subtract(A=@n80.XYZ, B=@n81), B=Constant3Vector() {constant=<rgba 1 1 0 0>})
   B=ScalarParameter() {defaultValue=1, parameterName=CameraFadeMaskPush}
  )
  n83 = AppendVector(A=Constant2Vector(), B=ScalarParameter() {defaultValue=-200, parameterName=CameraFadeMaskSink})
  n84 = Add(
   A=Add(A=Add(A=@n73, B=@n47), B=AppendVector(A=@n74, B=@n75))
   B=Multiply(A=SphereMask(A=@n76, B=@n77.XYZ, Radius=@n78, Hardness=@n79), B=Add(A=@n82, B=@n83))
  )
  n85 = Constant()
  n86 = VectorParameter() {parameterName=ShadCol, defaultValue=<rgba 0.28 0.2029 0.112 1>}
  n87 = Multiply(A=@n8, B=ScalarParameter() {defaultValue=1, parameterName=Tip_BrightnessSand}) {constB=2}
  n88 = Multiply(A=@n10, B=ScalarParameter() {defaultValue=1.15, parameterName=Tip_BrightnessGrass}) {constB=2}
  n89 = LinearInterpolate(
   A=@n7.GrassColor
   B=LinearInterpolate(A=LinearInterpolate(A=@n9, B=@n87, Alpha=@n17.G), B=LinearInterpolate(A=@n11, B=@n88, Alpha=@n17.G), Alpha=@n16.R)
   Alpha=MaterialFunctionCall() {materialFunction=/FigmentEnvironment/Migrated_From_LM_7_00/Athena/Environments/Landscape/MaterialFunctions/MF_Athena_BiomeMask}.Swamp
  )
  n90 = Multiply(A=@n89, B=@n17.R)
  n91 = LinearInterpolate(
   A=@n90
   B=LinearInterpolate(
    A=VectorParameter() {parameterName=LeafColor_1, defaultValue=<rgba 5 2.9558 0 1>}.RGB
    B=VectorParameter() {parameterName=LeafColor, defaultValue=<rgba 0.0416 0.1 0 1>}.RGB
    Alpha=@n28
   )
   Alpha=@n27.A
  )
  n92 = StaticSwitchParameter(
   True=LinearInterpolate(A=LinearInterpolate(A=@n90, B=@n91, Alpha=@n29.Result), B=@n86.RGB, Alpha=@n31.A)
   False=LinearInterpolate(A=@n89, B=@n86.RGB, Alpha=Multiply(A=@n31.A, B=@n31.A))
  ) {parameterName=IsShrub}
  MakeMaterialAttributes!(
   BaseColor=@n92
   Metallic
   Specular=LinearInterpolate(
    A=ScalarParameter() {defaultValue=0.02, parameterName=Spec_Min}
    B=ScalarParameter() {defaultValue=0.3, parameterName=Spec_Max}
    Alpha=@n17.G
   )
   Roughness=FeatureLevelSwitch(Default=@n18, ES2=@n19, ES3_1=@n19, SM4_REMOVED=@n18)
   Anisotropy
   EmissiveColor
   Opacity
   OpacityMask=StaticSwitchParameter(True=LinearInterpolate(A=@n30, B=@n23, Alpha=@n31.A), False=Add(A=@n30, B=@n45)) {parameterName=ShadowVertexMaskType}
   Normal=@n22
   Tangent
   WorldPositionOffset=FeatureLevelSwitch(Default=@n84, ES2=@n85, ES3_1=@n85, SM4_REMOVED=@n84)
   SubsurfaceColor
   ClearCoat
   ClearCoatRoughness
   AmbientOcclusion
   Refraction
   CustomizedUVs_0
   CustomizedUVs_1
   CustomizedUVs_2
   CustomizedUVs_3
   CustomizedUVs_4
   CustomizedUVs_5
   CustomizedUVs_6
   CustomizedUVs_7
   PixelDepthOffset
   ShadingModel
   Displacement
  )
  MaterialFunctionCall(Value=@n40) {materialFunction=&LinearSine}
  MaterialFunctionCall(In=@n17.G, Contrast=ScalarParameter() {parameterName=GradientContrast}) {materialFunction=&CheapContrast}
  Multiply(A=@n7.GrassNoColor, B=@n10)
  n93 = ScalarParameter() {defaultValue=1, parameterName=TempAAScale}
  Multiply(A=Constant2Vector() {g=2, r=1}, B=@n93)
  Multiply(A=ComponentMask(None=VectorParameter() {parameterName=TempAATextureSize_2, defaultValue=<rgba 1 1 0 1>}.RGB) {g=true, r=true}, B=@n93)
  Multiply(A=@n42, B=VectorParameter() {parameterName=TestNrm, defaultValue=<rgba -1 -1 1 1>}.RGB)
  Multiply(
   A=OneMinus(None=TextureSampleParameter2D() {parameterName=MainTexture_1, texture=&T_DesertGrass_01}.R)
   B=@n21
  )
  OneMinus(None=@n48)
  Panner(
   Coordinate=@n1.Result
   Time=Time()
   Speed=ComponentMask(None=VectorParameter() {parameterName=TempAATextureSize_1, defaultValue=<rgba 2 2 0 1>}.RGB) {g=true, r=true}
  )
  Power(Base=@n3, Exp=ScalarParameter() {defaultValue=1, parameterName=TempAAScale_1})
  Reroute(None=StaticSwitchParameter(True=Constant3Vector() {constant=<rgba 0 0 1 0>}, False=@n22) {parameterName=UseVertexNormalWPO})
  Reroute(None=@n17.A)
  ScalarParameter() {defaultValue=3, parameterName=SnowGrass_LowerMaskOffset}
  ScalarParameter() {defaultValue=200, parameterName=PlayerBendRadius_1}
  ScalarParameter() {defaultValue=-50, parameterName=PlayerBendStrength}
  StaticSwitchParameter(True=VertexNormalWS(), False=Constant3Vector() {constant=<rgba 0 0 1 0>}) {defaultValue=true, parameterName=UseVertexNormals}
  n94 = PerInstanceRandom()
  n95 = VectorParameter() {parameterName=ColorA, defaultValue=<rgba 0.675 0 0 1>}
  n96 = VectorParameter() {parameterName=ColorB, defaultValue=<rgba 0 0.745 0 1>}
  n97 = Constant() {r=0.25}
  n98 = VectorParameter() {parameterName=ColorC, defaultValue=<rgba 0 0.1854 0.725 1>}
  n99 = VectorParameter() {parameterName=ColorD, defaultValue=<rgba 1 0.3166 0 1>}
  n100 = If(
   A=@n94
   B=Constant() {r=0.75}
   "A > B"=@n95.RGB
   "A == B"=@n95.RGB
   "A < B"=If(
    A=@n94
    B=Constant() {r=0.5}
    "A > B"=@n96.RGB
    "A == B"=@n96.RGB
    "A < B"=If(A=@n94, B=@n97, "A > B"=@n98.RGB, "A == B"=@n98.RGB, "A < B"=@n99.RGB)
   )
  )
  StaticSwitchParameter(
   True=MaterialFunctionCall(
    1=Constant3Vector() {constant=<rgba 0.2243 0 1 1>}
    2=Constant3Vector() {constant=<rgba 0 1 0 0>}
    3=Constant3Vector() {constant=<rgba 0 0 1 0>}
    4=Constant3Vector() {constant=<rgba 1 0.5 0 1>}
    A=PerInstanceRandom()
   ) {materialFunction=&Lerp_Multiple_Float3}.Lerp 3 Inputs
   False=StaticSwitchParameter(True=StaticSwitchParameter(True=@n100, False=@n95.RGB) {defaultValue=true, parameterName=UseRandom}, False=@n92) {parameterName=DebugDiffuse}
  ) {parameterName=DebugDiffuseMulti}
  TextureSample() {samplerType=SAMPLERTYPE_LinearGrayscale, texture=&Good64x64TilingNoiseHighFreq}
  TextureSample() {texture=/Engine/MapTemplates/TilingNoise05}
  VectorParameter() {parameterName=TempAATextureSize, defaultValue=<rgba 64 64 0 1>}
  VectorParameter() {parameterName=DebugDiff, defaultValue=<rgba 0.985 0 0.0889 1>}
 ]
}
