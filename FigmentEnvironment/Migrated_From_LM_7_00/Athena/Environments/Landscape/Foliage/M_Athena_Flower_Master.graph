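# Masked, two-sided master material for instanced Athena landscape flowers, colored and cut out by a flower mask with per-instance and per-element randomness.
# Wind sway rotates Pivot Painter elements using panned swirl noise, with speed and power blended toward storm values by a storm mask.
# Player proximity bending layers onto the wind except at Low quality.
# Static switches add snow-based lowering and a blended normal map; per-instance fade scales the opacity mask.
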Material /FigmentEnvironment/Migrated_From_LM_7_00/Athena/Environments/Landscape/Foliage/M_Athena_Flower_Master
graph {
 properties {
  bTangentSpaceNormal=false
  bUseMaterialAttributes=true
  bUsedWithInstancedStaticMeshes=true
  blendMode=BLEND_Masked
  pixelDepthOffsetMode=PDOM_Legacy
  twoSided=true
  usedShadingModels=Default Lit
 }
 nodes [
  n1 = MaterialFunctionCall!(Wind Vector, Optimized for Foliage Placement=StaticBool() {value=true}, Optimized Vector=Constant3Vector() {constant=<rgba 0 0 1 0>}) {materialFunction=/Engine/Functions/Engine_MaterialFunctions02/PivotPainter/PivotPainter_PerObjectFoliageData}
  Add(A=@n1.Random Value Per Element) {constB=-0.35, desc=Use to sway odds}
  Clamp(None=MaterialFunctionCall(A=PerInstanceRandom()) {materialFunction=&Lerp_Multiple_Float3}.Lerp 3 Inputs)
  CollectionParameter() {
   collection=/FigmentEnvironment/Migrated_From_LM_7_00/Packages/Fortress_SharedMaterials/GlobalMaterialParameters/FortniteMaterialParameters
   parameterName=PlayerDirection
  }
  CollectionParameter() {
   collection=&FortniteMaterialParameters
   parameterName=PlayerDirection
  }
  CollectionParameter() {collection=&/MC_Athena_Terrain, parameterName=PlayerVelocity}
  CollectionParameter() {collection=&MC_Athena_Terrain, parameterName=WindIntensityStu}
  n2 = TextureObjectParameter() {
   parameterName=Flower_Mask
   samplerType=SAMPLERTYPE_LinearColor
   texture=&/T_Athena_Flowers_01
  }
  n3 = Add(A=PerInstanceRandom()) {constB=0.35, desc=Use to sway odds}
  n4 = MaterialFunctionCall!(In=@n2, Instance_RandomAlpha=Clamp(None=Round(None=@n3)), Element_RandomAlpha=@n1.Random Value Per Element) {materialFunction=/FigmentEnvironment/Migrated_From_LM_7_00/Athena/Environments/Landscape/MaterialFunctions/MF_Athena_FlowerColors}
  n5 = ScalarParameter() {defaultValue=1200, group=WPO, parameterName=WindWaveScale}
  n6 = ComponentMask(None=WorldPosition().XYZ) {g=true, r=true}
  n7 = Time()
  n8 = MaterialFunctionCall() {materialFunction=/FigmentEnvironment/Migrated_From_LM_7_00/Athena/Environments/Landscape/MaterialFunctions/MF_StormMask}
  n9 = LinearInterpolate(
   A=ScalarParameter() {defaultValue=0.04, group=WPO, parameterName=WindSpeed}
   B=ScalarParameter() {defaultValue=0.75, group=WPO, parameterName=WindSpeed_Storm}
   Alpha=@n8.Result
  )
  n10 = TextureSample(UVs=Panner(Coordinate=Divide(A=@n6, B=@n5), Time=Multiply(A=@n7, B=@n9)) {bFractionalPart=true, speedY=1}) {samplerType=SAMPLERTYPE_Normal, texture=/FigmentEnvironment/Migrated_From_LM_7_00/Environments/Foliage/Textures/T_SwirlNoise_N}
  n11 = ScalarParameter() {defaultValue=0.3257, group=WPO, parameterName=WindPower}
  n12 = LinearInterpolate(A=@n11, B=ScalarParameter() {defaultValue=0.75, group=WPO, parameterName=WindPower_Storm}, Alpha=@n8.Result)
  n13 = AppendVector(A=ComponentMask(None=Multiply(A=@n10.RGB, B=@n12)) {g=true, r=true}, B=Constant() {r=1})
  n14 = MaterialFunctionCall(
   VectorToTransform=@n13
   BasisX=Transform(None=Constant3Vector() {constant=<rgba 1 0 0 0>}) {transformSourceType=TRANSFORMSOURCE_Instance}
   BasisY=Transform(None=Constant3Vector() {constant=<rgba 0 1 0 0>}) {transformSourceType=TRANSFORMSOURCE_Instance}
   BasisZ=Transform(None=Constant3Vector() {constant=<rgba 0 0 1 0>}) {transformSourceType=TRANSFORMSOURCE_Instance}
  ) {materialFunction=&Transform3x3Matrix}
  n15 = TextureSample(UVs=Divide(A=@n6, B=ScalarParameter() {defaultValue=1000, group=WPO, parameterName=VarianceScale})) {samplerType=SAMPLERTYPE_LinearColor, texture=/FigmentEnvironment/Migrated_From_LM_7_00/Environments/Foliage/Textures/T_FieldGrass_01_M}
  n16 = Subtract(A=WorldPosition().XYZ, B=NamedRerouteDeclaration(None=@n1.Piv Position) {name=Pivot, nodeColor=<rgba 0 0.9531 1 1>})
  n17 = Subtract(A=WorldPosition().XYZ, B=Multiply(A=@n1.Element X-Axis, B=DotProduct(A=@n1.Element X-Axis, B=@n16)))
  n18 = RotateAboutAxis(
   NormalizedRotationAxis=Normalize(VectorInput=StaticSwitchParameter(True=@n14.Transformed Vector, False=@n13) {parameterName=UseAlignmentFix})
   RotationAngle=Multiply(A=@n15.R, B=ScalarParameter() {defaultValue=0.4, group=WPO, parameterName=AmbientWindAmount})
   PivotPoint=@n17
   Position=WorldPosition().XYZ
  )
  n19 = MaterialFunctionCall!(
   With RTT Features=Add(A=@n18, B=@n17)
   WithOut RTT Features=@n18
   With RTT Features V2
   WithOut RTT Features V2
   With RTT Features Att
   WithOut RTT Features Att
  ) {materialFunction=/FigmentEnvironment/Migrated_From_LM_7_00/WaterAndWind/EnableOrDisableWind/Fortnite_RenderToTextureEnabled}
  n20 = CollectionParameter() {
   collection=&FortniteMaterialParameters
   parameterName=PlayerPosition
  }
  n21 = MaterialFunctionCall!(
   PlayerPosition=@n20
   PlayerRadiusMax=ScalarParameter() {defaultValue=200, parameterName=PlayerBendRadiusMax}
   BendStrength=ScalarParameter() {defaultValue=-10, parameterName=BendStrength}
  ) {materialFunction=/FigmentEnvironment/Migrated_From_LM_7_00/Athena/Environments/Landscape/MaterialFunctions/MF_AthenaFoliagePlayerInteract}
  n22 = LinearInterpolate(A=@n19.V3, B=@n21.Rotation, Alpha=@n21.Rotation)
  n23 = QualitySwitch(Default=@n22, Low=@n19.V3, High=@n22, Medium=@n22)
  n24 = NamedRerouteUsage()
  n25 = OneMinus(None=Clamp(None=MaterialFunctionCall!(
     UseHeightLerp
     UV_WorldMask
     Use_ExtraSnowMask
     LandscapeBlend_Height
     Landscape_RockMask
     Snow_Mask
     Simple_UseWorldZNormal
     Simple_WorldNormal
     WorldPosition=@n24
    ) {materialFunction=/Game/Athena/Environments/Landscape/MaterialFunctions/Snow/MF_Athena_Holiday_SnowFall_S7}.SnowMask))
  n26 = Multiply(
   A=AppendVector(A=Constant2Vector(), B=ScalarParameter() {defaultValue=-100, parameterName=Snow_GrassLowerAmount})
   B=OneMinus(None=Power(Base=@n25))
  )
  n27 = VertexNormalWS()
  n28 = Power(
   Base=ComponentMask(None=TextureCoordinate() {coordinateIndex=1}) {g=true}
   Exp=ScalarParameter() {defaultValue=1, parameterName=VertNrmBlndAmt_3}
  )
  n29 = ScalarParameter() {defaultValue=1, parameterName=VertNrmBlndAmt_2}
  n30 = LinearInterpolate(
   A=Transform(None=TextureSample() {samplerType=SAMPLERTYPE_Normal, texture=/FigmentEnvironment/Migrated_From_LM_7_00/Environments/Foliage/Textures/T_WavyGrass_01_N}.RGB)
   B=@n27
   Alpha=Clamp(None=Multiply(A=@n28, B=@n29))
  )
  MakeMaterialAttributes!(
   BaseColor=@n4.Result
   Metallic
   Specular=ScalarParameter() {defaultValue=0.05, parameterName=Spec}
   Roughness=ScalarParameter() {defaultValue=1, parameterName=Rough}
   Anisotropy
   EmissiveColor
   Opacity
   OpacityMask=Multiply(A=@n4.Alpha, B=PerInstanceFadeAmount())
   Normal=StaticSwitchParameter(True=QualitySwitch(Default=@n30, Low=@n27, High=@n30, Medium=@n30), False=@n27) {parameterName=UseNormal}
   Tangent
   WorldPositionOffset=StaticSwitchParameter(True=Add(A=@n23, B=@n26), False=@n23) {group=Snow, parameterName=UseSnowLegacy}
   SubsurfaceColor
   ClearCoat
   ClearCoatRoughness
   AmbientOcclusion
   Refraction
   CustomizedUVs_0
   CustomizedUVs_1
   CustomizedUVs_2
   CustomizedUVs_3
   CustomizedUVs_4
   CustomizedUVs_5
   CustomizedUVs_6
   CustomizedUVs_7
   PixelDepthOffset
   ShadingModel
   Displacement
  )
  MaterialFunctionCall() {materialFunction=/Engine/Functions/Engine_MaterialFunctions02/PivotPainter/PivotPainter_PerObjectData}
  MaterialFunctionCall() {materialFunction=&ObjectPivotPoint}
  MaterialFunctionCall() {materialFunction=&PivotPainter_PerObjectData}
  Multiply(A=@n11)
  NamedRerouteUsage()
  n31 = MaterialFunctionCall!(
   "World Position (See Note)"=MaterialFunctionCall() {materialFunction=&PivotPainter_PerObjectData}.Pivot Position
   Max Rotation=ScalarParameter() {defaultValue=0.3, parameterName=Rotation Amount}
  ) {materialFunction=/FigmentEnvironment/Migrated_From_LM_7_00/WaterAndWind/MaterialFunctions/Fortnite_SampleRealTimeWindData}
  n32 = MaterialFunctionCall(Float2=TextureCoordinate() {coordinateIndex=1}) {materialFunction=&BreakOutFloat2Components}
  RotateAboutAxis(
   NormalizedRotationAxis=@n31.Rotation Axis
   RotationAngle=Multiply(A=@n31.Rotation Angle, B=OneMinus(None=@n32.G))
   PivotPoint=@n17
   Position=Add(A=@n18, B=WorldPosition().XYZ)
  )
  ScalarParameter() {defaultValue=3, parameterName=SnowGrass_LowerMaskOffset}
  Subtract(A=WorldPosition().XYZ)
 ]
}
